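; VividBronco thrown heal-bomb projectile (Chug Splash style) that splashes healing and an attack-speed effect onto hit pawns.
; Hit or water entry triggers a one-shot path: water-in-radius, re-enabled instigator overlap, server explosion, client decal.
; The explosion applies effects once per actor, reports squad/teammate heals to quests, plays FX cues and sends the instigator a gameplay event.
; It then disables collision and kills itself after the heal-pool duration.
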(event EventOnTouched
  (params (OtherActor :type Actor%20Object%20Reference) (OtherComp :type Primitive%20Component%20Object%20Reference) (HitResult :type Hit%20Result%20Structure%20%28by%20ref%29) (bIsOverlap :type Boolean))
  (bind (_bblockinghit _binitialoverlap _time _distance _location _impactpoint _normal _impactnormal _physmat _hitactor _hitcomponent) (Collision|BreakHitResult HitResult))
  ;[1 1 1] EVENT -> On Touched: No Stick Check
  (|CallParentFunction OtherActor OtherComp HitResult bIsOverlap)
  (if (not (Components|ComponentHasTag _hitcomponent "NoStick"))
    (Variables|Projectile|SetbStopSimulatingOnHit true)))

(event OnEnterWaterEvent(WaterInteractionComponent)
  (params (WaterBody :type Fort%20Water%20Body%20Actor%20Object%20Reference) (WaterInteractionComponent :type Fort%20Water%20Interaction%20Component%20Object%20Reference) (bIsFirstBody :type Boolean))
  ;[1 1 1] On enter water call the the same path as the hit event but with a different hit normal and location.
  (Variables|BPrjHealBomb|SetHitLocation (Transformation|GetActorLocation))
  (Variables|BPrjHealBomb|SetHitNormal "0, 0,1.000000")
  ;[0.7116 0.838 1] EVENT-> OnStop: Explode
  (Utilities|FlowControl|DoOnce false
    (:Completed
      (CallFunction|ApplyWaterinRadiusifCurieEnabled (stale Collision|EventHit Hit))
      (Collision|SetIgnoreInstigatorCollision false) ; Need to do this or the instigator cannot benefit from the overlap heal
      (Components|Movement|SetUpdatedComponent (Variables|Default|GetRotatingMovement) 0)
      (Networking|SwitchHasAuthority
        (:Authority
          (Delivery|DoExplosion))
        (:Remote
          (bind _asfort_pawn (Utilities|Casting|CastToFortPawn (stale Collision|EventHit Other))
            (:then
              (bind _returnvalue (Game|SpawnActorfromClass "/VividRazorConsumablesGameplay/Assets/FX/B_Decal_VividBronco.B_Decal_VividBronco_C" (Math|Conversions|ToTransform(Vector) (Transformation|GetWorldLocation (Variables|/Script/Engine.Character|GetMesh _asfort_pawn))) "Undefined" "MultiplyWithRoot" 0 0))
              (Transformation|SetWorldRotation (Variables|Transformation|GetRootComponent _returnvalue) (Math|Rotator|MakeRotator (+ (.roll "90.000000, 0, 0") (Math|Random|RandomFloatinRange 0.0 3.6e+02)) (.pitch "90.000000, 0, 0") (.yaw "90.000000, 0, 0")) false false))
            (:CastFailed
              (bind _returnvalue_2 (Game|SpawnActorfromClass "/VividRazorConsumablesGameplay/Assets/FX/B_Decal_VividBronco.B_Decal_VividBronco_C" (Math|Conversions|ToTransform(Vector) (Variables|BPrjHealBomb|GetHitLocation)) "Undefined" "MultiplyWithRoot" 0 0))
              (Transformation|SetWorldRotation (Variables|Transformation|GetRootComponent _returnvalue_2) (Math|Rotator|MakeRotator (+ (.roll (Math|Rotator|MakeRotfromX (Variables|BPrjHealBomb|GetHitNormal))) (Math|Random|RandomFloatinRange 0.0 3.6e+02)) (* (.pitch (Math|Rotator|MakeRotfromX (Variables|BPrjHealBomb|GetHitNormal))) -1.0) (.yaw (Math|Rotator|MakeRotfromX (Variables|BPrjHealBomb|GetHitNormal)))) false false))))))))

(event Collision|EventHit
  (params (MyComp :type Primitive%20Component%20Object%20Reference) (Other :type Actor%20Object%20Reference) (OtherComp :type Primitive%20Component%20Object%20Reference) (bSelfMoved :type Boolean) (HitLocation :type Vector :default "0, 0, 0") (HitNormal :type Vector :default "0, 0, 0") (NormalImpulse :type Vector :default "0, 0, 0") (Hit :type Hit%20Result%20Structure%20%28by%20ref%29))
  ;[0.7116 0.838 1] EVENT-> OnStop: Explode
  (Variables|BPrjHealBomb|SetHitLocation HitLocation)
  (Variables|BPrjHealBomb|SetHitNormal HitNormal)
  (Utilities|FlowControl|DoOnce false
    (:Completed
      (CallFunction|ApplyWaterinRadiusifCurieEnabled Hit)
      (Collision|SetIgnoreInstigatorCollision false) ; Need to do this or the instigator cannot benefit from the overlap heal
      (Components|Movement|SetUpdatedComponent (Variables|Default|GetRotatingMovement) 0)
      (Networking|SwitchHasAuthority
        (:Authority
          (Delivery|DoExplosion))
        (:Remote
          (bind _asfort_pawn (Utilities|Casting|CastToFortPawn Other)
            (:then
              (bind _returnvalue (Game|SpawnActorfromClass "/VividRazorConsumablesGameplay/Assets/FX/B_Decal_VividBronco.B_Decal_VividBronco_C" (Math|Conversions|ToTransform(Vector) (Transformation|GetWorldLocation (Variables|/Script/Engine.Character|GetMesh _asfort_pawn))) "Undefined" "MultiplyWithRoot" 0 0))
              (Transformation|SetWorldRotation (Variables|Transformation|GetRootComponent _returnvalue) (Math|Rotator|MakeRotator (+ (.roll "90.000000, 0, 0") (Math|Random|RandomFloatinRange 0.0 3.6e+02)) (.pitch "90.000000, 0, 0") (.yaw "90.000000, 0, 0")) false false))
            (:CastFailed
              (bind _returnvalue_2 (Game|SpawnActorfromClass "/VividRazorConsumablesGameplay/Assets/FX/B_Decal_VividBronco.B_Decal_VividBronco_C" (Math|Conversions|ToTransform(Vector) (Variables|BPrjHealBomb|GetHitLocation)) "Undefined" "MultiplyWithRoot" 0 0))
              (Transformation|SetWorldRotation (Variables|Transformation|GetRootComponent _returnvalue_2) (Math|Rotator|MakeRotator (+ (.roll (Math|Rotator|MakeRotfromX (Variables|BPrjHealBomb|GetHitNormal))) (Math|Random|RandomFloatinRange 0.0 3.6e+02)) (* (.pitch (Math|Rotator|MakeRotfromX (Variables|BPrjHealBomb|GetHitNormal))) -1.0) (.yaw (Math|Rotator|MakeRotfromX (Variables|BPrjHealBomb|GetHitNormal)))) false false))))))))

(event EventOnExploded
  (params (HitActors :type Array%20of%20Actor%20Object%20References) (HitResults :type Array%20of%20Hit%20Result%20Structures))
  ;[1 1 1] EVENT -> OnExploded - Apply Health/Shields and FX
  (SoundIndicator|ShowSoundIndicator (Variables|Default|GetFortSoundIndicator) (Variables|BPrjHealBomb|GetHitLocation) "/Game/Athena/Sounds/Items/Chug_Bomb/SoundIndicator_Chug_Bomb_Explode_Cue.SoundIndicator_Chug_Bomb_Explode_Cue" 1.0 0)
  (CallFunction|HealHitPlayer HitActors)
  (Development|SetHiddeninGame (Variables|Default|GetSK_ChugBomb) true false)
  (Components|Activation|Deactivate (Variables|Default|GetProj_TrailFX))
  (Ability|SendGameplayEventtoActor (Variables|Actor|GetInstigator) "(TagName=\"Event.Athena.ChillBronco.Hit\")" (Utilities|Struct|MakeGameplayEventData :ContextHandle "()" :InstigatorTags "(GameplayTags=)" :TargetTags "(GameplayTags=)" :EventMagnitude 0.0 :TargetData (Ability|TargetData|AbilityTargetDatafromActorArray (Variables|BPrjHealBomb|GetHitActors))))
  (CallFunction|EnableCollision)
  (Audio|Components|Audio|FadeOut (Variables|Default|GetAudio_InAirLoop) 0.5 0.0 "Linear" :!class /Script/Engine.AudioComponent))

(event Custom|Heal%20Hit%20Player
  (params (HitActors :type Array%20of%20Actor%20Object%20References))
  (bind _instigator (Variables|Actor|GetInstigator))
  (bind _returnvalue (Ability|GetAbilitySystemComponent _instigator))
  ;[1 1 1] Heal Players
  (for _array_element HitActors
    (if (not (Utilities|Array|ContainsItem (Variables|Default|GetHealed%20Actors) _array_element))
      (bind _asfort_pawn (Utilities|Casting|CastToFortPawn _array_element))
      (Utilities|Array|Add (Variables|Default|GetHealed%20Actors) _asfort_pawn)
      (bind _asfort_ability_system_component (Utilities|Casting|CastToFortAbilitySystemComponent _returnvalue))
      (GameplayEffects|ApplyGameplayEffectToTarget _returnvalue "/VividRazorConsumablesGameplay/Gameplay/GE_Athena_VividBronco_Healing.GE_Athena_VividBronco_Healing_C" (Ability|GetAbilitySystemComponent _asfort_pawn) 0.0 (GameplayEffects|MakeFortEffectContext _asfort_ability_system_component 0 self 0 0))
      (GameplayEffects|ApplyGameplayEffectToTarget :self _returnvalue :GameplayEffectClass "/GoldRushGameplay/Gameplay/GE_GoldRush_AttackSpeedMult.GE_GoldRush_AttackSpeedMult_C" :Target (Ability|GetAbilitySystemComponent _asfort_pawn) :Level 0.0)
      (Class|FortPawn|LogChugSplashEvent _asfort_pawn)
      (CallFunction|QuestCheckSquadMember _asfort_pawn))))

(event Custom|KillDelay
  ;[1 1 1] Kill Projectile
  (Delivery|Kill))

(event Custom|EnableCollision
  ;[1 1 1] Enable Collision
  (Collision|SetCollisionEnabled (Variables|Collision|GetCapsuleComponent) "NoCollision") ; Need this or collision with projectile can still happen after it hits
  (Utilities|Time|SetTimerbyEvent (EventDispatchers|CreateEvent :!event Custom|KillDelay) (ScalableFloat|GetValueatLevel (Variables|BPrjHealBomb|GetHealPool_Duration))))

(fn GetAffiliationsToShowFor
  (return (Utilities|Array|MakeArray "Friendly" "Hostile")))

(fn QuestCheckSquadMember
  (params (PlayerPawnHit :type Fort%20Pawn%20Object%20Reference))
  (bind _returnvalue (Actor|GetOwner))
  ;[1 1 1] Save off references.
  (bind _asfort_player_pawn_athena (Utilities|Casting|CastToFortPlayerPawnAthena _returnvalue))
  (bind _output_get (LocalVariables|SetOwning%20Player%20PC (Fortnite|GetFortPlayerControllerfromActor _asfort_player_pawn_athena)))
  (if (and (Utilities|IsValid _output_get :!class /Script/Engine.KismetSystemLibrary) (Utilities|IsValid (Variables|GetPlayerPawnHit) :!class /Script/Engine.KismetSystemLibrary))
    (Variables|Default|SetOwning%20Player%20Quest%20Manager (Quests|GetQuestManager _output_get "Invalid"))
    ;[1 1 1] Filter owner out of  checks.
    (if (Utilities|NotEqual(Object) _returnvalue (Variables|GetPlayerPawnHit))
      (bind _asfort_player_state_athena (Utilities|Casting|CastToFortPlayerStateAthena (Variables|/Script/Engine.Controller|GetPlayerState (Pawn|GetController (Variables|GetPlayerPawnHit)))))
      ;[1 1 1] Set friendliness level bools.
      (LocalVariables|SetIs%20Squad%20Member (Class|FortPlayerStateAthena|IsinSquad _asfort_player_state_athena))
      (Development|PrintString (Utilities|String|Append "Target is in squad =" (Utilities|String|ToString(Boolean) (LocalVariables|GetIs%20Squad%20Member))) false false)
      (LocalVariables|SetIs%20Team%20Mate (Team|OnSameTeam (LocalVariables|GetOwning%20Player%20PC) _asfort_player_state_athena))
      ;[1 1 1] Send squad and/or teammate healed quest stat events.
      (if (LocalVariables|GetIs%20Squad%20Member)
        (Quest|SendComplexCustomStatEvent (Variables|Default|GetOwning%20Player%20Quest%20Manager) 0 (Variables|Default|GetHeal%20Squadmate%20TC) "(GameplayTags=)" 1)
        (elif (LocalVariables|GetIs%20Team%20Mate)
          (Quest|SendComplexCustomStatEvent (Variables|Default|GetOwning%20Player%20Quest%20Manager) 0 (Variables|Default|GetHeal%20Teammate%20TC) "(GameplayTags=)" 1))))))

(fn ShouldShowSoundIndicator
  (params (PlayerController :type Fort%20Player%20Controller%20Object%20Reference))
  (return true))

(fn ConstructionScript
  (|CallParentFunction)
  (Components|Movement|SetUpdatedComponent (Variables|Default|GetRotatingMovement) (Variables|Collision|GetCapsuleComponent)))
