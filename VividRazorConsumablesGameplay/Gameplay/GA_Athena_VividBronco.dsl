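; Gameplay ability for the VividBronco thrown consumable, extending a parent projectile-throwing ability.
; It shows the trajectory only on use and overrides projectile spawning to add configurable roll and pitch offsets to the spawned projectile's rotation, keeping its yaw.
; When a projectile spawns, it sends the shared ChugAndChillBronco quest stat event through the avatar's player controller quest manager, if that controller is valid.
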(event Ability|EventActivateAbility
  (|CallParentFunction) ; This node is disabled and will not be called.\nDrag off pins to build functionality.
)

(event Ability|EventOnEndAbility
  (params (bWasCancelled :type Boolean))
  (|CallParentFunction bWasCancelled) ; This node is disabled and will not be called.\nDrag off pins to build functionality.
)

(event ProjectileLifecycle|EventOnProjectileSpawned
  (params (NewSpawnedProjectile :type Fort%20Projectile%20Base%20Object%20Reference))
  (bind _returnvalue (Ability|GetAvatarActorfromActorInfo))
  (bind _returnvalue_2 (Fortnite|GetFortPlayerControllerfromActor _returnvalue))
  (|CallParentFunction NewSpawnedProjectile)
  (Utilities|IsValid _returnvalue_2
    (:Is%20Valid
      (Quest|SendComplexCustomStatEvent (Quests|GetQuestManager _returnvalue_2 "Invalid") 0 "(GameplayTags=((TagName=\"Athena.Quests.Ability.Shared.ChugAndChillBronco\")))" "(GameplayTags=((TagName=\"Athena.Quests.Ability.Shared.ChugAndChillBronco\")))" 1))))

(fn ShouldOnlyShowTrajectoryOnUse
  (return true))

(fn SpawnThrownProjectile
  (params (ProjectileClassToSpawn :type Fort%20Projectile%20Base%20Class%20Reference) (ProjectileSpawnLocation :type Vector%20%28by%20ref%29 :default "0, 0, 0") (ProjectileSpawnDirection :type Rotator%20%28by%20ref%29 :default "0, 0, 0") (HitEffectContainerSpec :type Fort%20Gameplay%20Effect%20Container%20Spec%20Structure%20%28by%20ref%29) (ExplosionEffectContainerSpec :type Fort%20Gameplay%20Effect%20Container%20Spec%20Structure%20%28by%20ref%29) (ProjectileInitialSpeed :type Float%20%28double-precision%29) (ThrownProjectileGravityScale :type Float%20%28double-precision%29))
  (bind _returnvalue (|CallParentFunction ProjectileClassToSpawn ProjectileSpawnLocation ProjectileSpawnDirection HitEffectContainerSpec ExplosionEffectContainerSpec ProjectileInitialSpeed ThrownProjectileGravityScale))
  (bind (_returnvalue_roll _returnvalue_pitch _returnvalue_yaw) (Math|Rotator|BreakRotator (Transformation|GetActorRotation _returnvalue)))
  (Transformation|SetActorRotation _returnvalue (Math|Rotator|MakeRotator (+ _returnvalue_roll (Variables|Default|GetAdded_X_SpawnRotation)) (+ _returnvalue_pitch (Variables|Default|GetAdded_Y_SpawnRotation)) _returnvalue_yaw) false)
  (return _returnvalue))
